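# Control Rig for the Sweet Teriyaki Red backpack that squashes and stretches its dyn_body bone based on jiggle-bone movement.
# The absolute local Z translation of dyn_jiggle, a child of dyn_body, is remapped into weights for two chained ModifyTransforms scale steps.
# Offsets above a middle threshold slightly stretch the body vertically, while offsets below it apply a stronger squash and widening.
# The rig only reads the jiggle bone and does not simulate its motion.
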ControlRigBlueprint /BRCosmetics/Accessories/FORT_Backpacks/Backpack_M_MED_Sweet_Teriyaki_Red/Meshes/M_MED_Sweet_Teriyaki_Red_Pack_CtrlRig
graphs [
 RigVMFunctionLibrary {links [], nodes []}
 RigVMModel {nodes [
   RigUnit_BeginExecution = &RigUnit_BeginExecution()
   n1 = /Script/RigVM.RigVMFunction_MathFloatAbs(Value=&RigUnit_GetTransform(Item=(Type=Bone,Name="dyn_jiggle"), Space=LocalSpace).Transform.Translation.Z) {is=struct RigVMUnitNode, template="Absolute::Execute(in Value,out Result)"}
   &RigUnit_ModifyTransforms(
    ItemToModify=((Item=(Type=Bone,Name="dyn_body"),Transform=(Rotation=(X=0,Y=0,Z=0,W=1),Translation=(X=0,Y=0,Z=0),Scale3D=(X=0.975,Y=1,Z=1.025))))
    Weight=&RigVMFunction_MathFloatRemap(SourceMaximum=8, SourceMinimum=4, TargetMinimum=0, Value=@n1.Result).Result
    ExecutePin=@RigUnit_BeginExecution.ExecutePin
   )
   > &RigUnit_ModifyTransforms(
    ItemToModify=((Item=(Type=Bone,Name="dyn_body"),Transform=(Rotation=(X=0,Y=0,Z=0,W=1),Translation=(X=0,Y=0,Z=0),Scale3D=(X=1.1,Y=1,Z=0.9))))
    Weight=&RigVMFunction_MathFloatRemap(SourceMaximum=2, SourceMinimum=4, TargetMinimum=0, Value=@n1.Result).Result
   )
  ]}
]
hierarchy [
 root {loc=<0 0 0>, rot=<0 0 0 1>}
 pack {parent=root, loc=<0 -0.7316 0>, rot=<0 0 0 1>}
 dyn_body {parent=root, loc=<0 7.9675 0>, rot=<0 0 0 1>}
 dyn_eye_l {parent=dyn_body, loc=<8.9757 -18.8231 7.6878>, rot=<-0.0755 0.3078 0.226 -0.9212>}
 dyn_fin_l1 {parent=dyn_body, loc=<11.7507 -16.5889 -1.6889>, rot=<-0.0706 -0.2905 0.2651 -0.9167>}
 dyn_fin_l2 {parent=dyn_fin_l1, loc=<2.7947 0 0>, rot=<0 -0.0279 0 -0.9996>}
 dyn_fin_r1 {parent=dyn_body, loc=<-11.7507 -16.5889 -1.6889>, rot=<-0.2905 -0.0706 0.9167 -0.2651>}
 dyn_fin_r2 {parent=dyn_fin_r1, loc=<2.7947 0 0>, rot=<0 -0.0279 0 -0.9996>}
 dyn_tail_1 {parent=dyn_body, loc=<0 -7.8322 -18.1203>, rot=<0 -0.7071 0 -0.7071>}
 dyn_tail_2 {parent=dyn_tail_1, loc=<3.1676 0 0>, rot=<0 0 0 -1>}
 dyn_eye_r {parent=dyn_body, loc=<-8.9757 -18.8231 7.6878>, rot=<0.3078 -0.0755 0.9212 -0.226>}
 dyn_jiggle {parent=dyn_body, loc=<0 0 0>, rot=<0 0 0 1>}
]
